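; Server-only analytics volume for a weapon bunker that records player activity around it, with client collision disabled.
; Box overlaps track unique humans, bots and teams, human dwell time, and bind pawn delegates for damage and deaths.
; Damage splits by whether the victim is inside, deaths require both pawns inside, and NotifyActorBunkerEntered records bunker entries.
; On EndGame, or EndPlay in editor, it submits one bunker-name-prefixed analytics event, once and never during replays.
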(event EventBeginPlay
  ;[0.01873 0.1667 0.03335] BeginPlay: On the server, bind delegate to the game state to know when the game phase changes.
  ; On the client, disable collision.
  (|CallParentFunction)
  (Utilities|SwitchNetMode
    (:Server
      (Class|FortAsyncActionCurrentPlaylistReady|PlaylistDataReadyAsync
        (:OnReady
          (bind _returnvalue (GameplayEventRouter|GetEventRouter "/Script/FortniteGame.FortGameState" self))
          (bind (_asynctaskproxy _payload) (GameplayEventRouter|StartListeningToStatefulEvent _returnvalue "/Script/FortniteGame.FortGamePhaseUpdatedEvent")
            (:then
              (Variables|Transient|SetGamePhaseUpdatedAsyncTask _asynctaskproxy))
            (:OnEventReceived
              _)
            (:OnSavedEventStateExists
              _))
          (CallFunction|HandleGamePhaseChanged (Utilities|Struct|BreakFortGamePhaseUpdatedEvent _payload)))))
    (:Client
      (Collision|SetCollisionEnabled (Variables|Default|GetBox) "NoCollision"))))

(event OnComponentBeginOverlap(Box)
  (params (OverlappedComponent :type Primitive%20Component%20Object%20Reference) (OtherActor :type Actor%20Object%20Reference) (OtherComp :type Primitive%20Component%20Object%20Reference) (OtherBodyIndex :type Integer) (bFromSweep :type Boolean) (SweepResult :type Hit%20Result%20Structure%20%28by%20ref%29))
  ;[0.03703 0.1469 0.1771] OnComponentBeginOverlap: Start tracking the player that has entered the volume while tracking values for analytics.
  (Utilities|SwitchNetMode
    (:Server
      (bind _asfort_player_pawn_athena (Utilities|Casting|CastToFortPlayerPawnAthena OtherActor)) ; Overlapped a player pawn.
      (bind _asfort_player_state_athena (Utilities|Casting|CastToFortPlayerStateAthena (Variables|/Script/Engine.Pawn|GetPlayerState _asfort_player_pawn_athena)))
      (if (Variables|/Script/Engine.PlayerState|GetbIsABot _asfort_player_state_athena)
        ;;[0.009605 0.05729 0.05043] Track bots and players separately.
        (Utilities|Array|AddUnique (Variables|Transient|GetUniquePlayerIds_Bot) (PlayerState|GetPlayerUniqueNetID _asfort_player_state_athena))
        (else
          (Utilities|Array|AddUnique (Variables|Transient|GetUniquePlayerIds_Human) (PlayerState|GetPlayerUniqueNetID _asfort_player_state_athena))
          (Utilities|Map|Add (Variables|Transient|GetUniquePlayerToTimeEnteredVolumeMap) (PlayerState|GetPlayerUniqueNetID _asfort_player_state_athena) (Fortnite|Sync|Time|GetServerWorldTimeSeconds))))
      ;;[0.1354 0.07014 0.03995] Track players that are currently in the volume.
      (Utilities|Array|AddUnique (Variables|Transient|GetOverlappingPlayerPawns) _asfort_player_pawn_athena)
      ;;[0.2395 1 0.5426] Track unique teams that enter the volume.
      (Utilities|Array|AddUnique (Variables|Transient|GetTeams) (Math|Conversions|ToInteger(Byte) (Team|GetActorTeam _asfort_player_state_athena)))
      ;;[0.01975 0.08333 0.06826] Setup delegates to track what the player is doing in the volume.
      (EventDispatchers|BindEventtoOnDidDamage _asfort_player_pawn_athena (EventDispatchers|CreateEvent :!event HandlePawnDidDamage))
      (EventDispatchers|BindEventtoOnFortPlayerDied _asfort_player_pawn_athena (EventDispatchers|CreateEvent :!event HandlePawnDied))
      (Fortnite|FortShippingLog self (Utilities|String|Append "OnComponentBeginOverlap: " (Fortnite|GetPlayerNameSafe _asfort_player_pawn_athena))))))

(event OnComponentEndOverlap(Box)
  (params (OverlappedComponent :type Primitive%20Component%20Object%20Reference) (OtherActor :type Actor%20Object%20Reference) (OtherComp :type Primitive%20Component%20Object%20Reference) (OtherBodyIndex :type Integer))
  ;[0.1042 0.03434 0.08598] OnComponentEndOverlap: Unbind delegates.
  (Utilities|SwitchNetMode
    (:Server
      (bind _asfort_player_pawn_athena (Utilities|Casting|CastToFortPlayerPawnAthena OtherActor))
      (Utilities|Array|RemoveItem (Variables|Transient|GetOverlappingPlayerPawns) _asfort_player_pawn_athena)
      (EventDispatchers|UnbindEventfromOnDidDamage _asfort_player_pawn_athena (EventDispatchers|CreateEvent :!event HandlePawnDidDamage))
      (EventDispatchers|UnbindEventfromOnFortPlayerDied _asfort_player_pawn_athena (EventDispatchers|CreateEvent :!event HandlePawnDied))
      (Fortnite|FortShippingLog self (Utilities|String|Append "OnComponentEndOverlap: " (Fortnite|GetPlayerNameSafe _asfort_player_pawn_athena)))
      (CallFunction|ComputeTotalDurationforLeavingPlayer _asfort_player_pawn_athena))))

(event EventEndPlay
  (params (EndPlayReason :type EEndPlayReason%20Enum :default "Destroyed"))
  ;[0.07292 0.04601 0.006518] EndPlay: Unbind delegates.
  (Utilities|FlowControl|Sequence
    (:then_0
      (Utilities|SwitchNetMode
        (:Server
          (Utilities|FlowControl|Sequence
            (:then_0
              (if (Editor|InEditororPIE)
                (CallFunction|SubmitAnalytics)
                (else
                  _))
              (bind _gamephaseupdatedasynctask (Variables|Transient|GetGamePhaseUpdatedAsyncTask)
                (:then
                  (GameplayEventRouter|StopListeningtoStatefulEvent _gamephaseupdatedasynctask))
                (:else)))
            (:then_1
              (for _array_element (Variables|Transient|GetOverlappingPlayerPawns)
                (Utilities|IsValid _array_element
                  (:Is%20Valid
                    (EventDispatchers|UnbindEventfromOnDidDamage _array_element (EventDispatchers|CreateEvent :!event HandlePawnDidDamage))
                    (EventDispatchers|UnbindEventfromOnFortPlayerDied _array_element (EventDispatchers|CreateEvent :!event HandlePawnDied))))))))))
    (:then_1
      (|CallParentFunction EndPlayReason))))

(fn BuildDataRegistryResolverScope
  (bind _self self)
  (bind _returnvalue (Minigame|GetMinigame _self :!class /Script/FortniteGame.MinigameBlueprintLibrary))
  ;[1 0.8475 0.2776] Gives the name of the volume as a scope to the data registry. Needed for calls to GetValueWithContext to work
  (Utilities|IsValid _returnvalue
    (:Is%20Valid
      (return (Utilities|Array|MakeArray (SpatialGameplay|GetSpatialGameplayInterfaceFNameforActor _returnvalue)) 0 true))
    (:Is%20Not%20Valid
      (return 0 0 false))))

(fn ComputeTotalDurationForLeavingPlayer
  (params (Pawn :type Pawn%20Object%20Reference))
  (bind _returnvalue (Fortnite|Sync|Time|GetServerWorldTimeSeconds))
  (bind _uniqueplayertototaldurationmap (Variables|Transient|GetUniquePlayerToTotalDurationMap))
  (bind _uniqueplayertotimeenteredvolumemap (Variables|Transient|GetUniquePlayerToTimeEnteredVolumeMap))
  ;[0.02231 0.1667 0.1584] ComputeTotalDurationForLeavingPlayer: Computes the total duration that the leaving player has spent in the volume.
  (Utilities|SwitchNetMode
    (:Server
      (Utilities|IsValid Pawn
        (:Is%20Valid
          (bind _asfort_player_state_athena (Utilities|Casting|CastToFortPlayerStateAthena (Variables|/Script/Engine.Pawn|GetPlayerState Pawn)))
          (if (not (Variables|/Script/Engine.PlayerState|GetbIsABot _asfort_player_state_athena))
            (if (.ReturnValue (Utilities|Map|Find _uniqueplayertotimeenteredvolumemap (PlayerState|GetPlayerUniqueNetID _asfort_player_state_athena)))
              (if (.ReturnValue (Utilities|Map|Find _uniqueplayertototaldurationmap (PlayerState|GetPlayerUniqueNetID _asfort_player_state_athena)))
                (Utilities|Map|Add (Variables|Transient|GetUniquePlayerToTotalDurationMap) (PlayerState|GetPlayerUniqueNetID _asfort_player_state_athena) (+ (Utilities|Map|Find _uniqueplayertototaldurationmap (PlayerState|GetPlayerUniqueNetID _asfort_player_state_athena)) (- _returnvalue (Utilities|Map|Find _uniqueplayertotimeenteredvolumemap (PlayerState|GetPlayerUniqueNetID _asfort_player_state_athena)))))
                (else
                  (Utilities|Map|Add (Variables|Transient|GetUniquePlayerToTotalDurationMap) (PlayerState|GetPlayerUniqueNetID _asfort_player_state_athena) (- _returnvalue (Utilities|Map|Find _uniqueplayertotimeenteredvolumemap (PlayerState|GetPlayerUniqueNetID _asfort_player_state_athena))))))
              (Utilities|Map|Remove (Variables|Transient|GetUniquePlayerToTimeEnteredVolumeMap) (PlayerState|GetPlayerUniqueNetID _asfort_player_state_athena)))))))))

(fn CreateAttributeName
  (params (Attribute :type String))
  ;[0.06757 0.1617 0.1667] Add the bunker name to the attribute name.
  (return (Utilities|String|Append (Variables|Default|GetBunkerName) " : " Attribute)))

(fn GetAveragePlayerGameplayTimeInVolume
  ;[0.008687 0.08909 0.09896] Returns the average time players spent in the volume.
  (bind _time (CallFunction|GetTotalPlayerGameplayTimeinVolume))
  (return (Math|Float|SafeDivide _time (Math|Conversions|ToFloat(Integer) (Utilities|Map|Length (Variables|Transient|GetUniquePlayerToTotalDurationMap))))))

(fn GetCacheContextOverride
  ;[0.4163 1 0.9801] We don't need to return anything here since we don't need to override the Cache Context.
  (return 0))

(fn GetFinisherOrDownerPawn
  (params (DyingPlayerPawn :type Fort%20Player%20Pawn%20Athena%20Object%20Reference))
  ;[0.06228 0.1458 0.1458] Returns the player reponsible for downing or eliminating the input player.
  (Utilities|FlowControl|Sequence
    (:then_0
      (bind _playerstate (Variables|/Script/Engine.Pawn|GetPlayerState DyingPlayerPawn)
        (:then
          (bind _asfort_player_state_athena (Utilities|Casting|CastToFortPlayerStateAthena _playerstate))
          (bind _returnvalue_finisherordowner (Utilities|Struct|BreakDeathInfo (Class|FortPlayerStateAthena|GetDeathInfo _asfort_player_state_athena)))
          (Utilities|IsValid _returnvalue_finisherordowner
            (:Is%20Valid
              (Utilities|IsValid (Fortnite|GetControllerfromActor _returnvalue_finisherordowner)
                (:Is%20Valid
                  (bind _asfort_player_pawn_athena (Utilities|Casting|CastToFortPlayerPawnAthena (Pawn|GetControlledPawn (Fortnite|GetControllerfromActor _returnvalue_finisherordowner))))
                  (return _asfort_player_pawn_athena))))))
        (:else)))
    (:then_1
      (return 0))))

(fn GetTotalPlayerGameplayTimeInVolume
  (bind _uniqueplayertototaldurationmap (Variables|Transient|GetUniquePlayerToTotalDurationMap))
  ;[0.01335 0.09896 0.09601] Sum up the time each unique player spent in the volume.
  (bind _keys (Utilities|Map|Keys _uniqueplayertototaldurationmap))
  (for _array_element _keys
    (LocalVariables|SetRetTime (+ (LocalVariables|GetRetTime) (Utilities|Map|Find _uniqueplayertototaldurationmap _array_element))))
  (return (LocalVariables|GetRetTime)))

(fn HandleGamePhaseChanged
  (params (GamePhase :type EAthenaGamePhase%20Enum))
  ;[0.0164 0.07935 0.1198] HandleGamePhaseChanged: Once we're in EndGame, submit the analytics.
  (Utilities|SwitchNetMode
    (:Server
      (if (Utilities|Enum|Equal(Enum) GamePhase "EndGame")
        (CallFunction|SubmitAnalytics)
        (bind _gamephaseupdatedasynctask (Variables|Transient|GetGamePhaseUpdatedAsyncTask)
          (:then
            (GameplayEventRouter|StopListeningtoStatefulEvent _gamephaseupdatedasynctask))
          (:else))))))

(fn HandlePawnDidDamage
  (params (DamagedActor :type Actor%20Object%20Reference) (Damage :type Float%20%28single-precision%29) (InstigatedBy :type Controller%20Object%20Reference) (DamageCauser :type Actor%20Object%20Reference) (HitLocation :type Vector) (FHitComponent :type Primitive%20Component%20Object%20Reference) (BoneName :type Name) (Momentum :type Vector))
  (bind (_asfort_athena_aibot_controller _bsuccess) (Utilities|Casting|CastToFortAthenaAIBotController InstigatedBy))
  (bind (_asfort_player_controller_athena _bsuccess_2) (Utilities|Casting|CastToFortPlayerControllerAthena InstigatedBy))
  (bind _playerstate (Variables|/Script/Engine.Controller|GetPlayerState InstigatedBy))
  (bind _damage (Variables|GetDamage))
  ;[0.02222 0.1579 0.1927] HandlePawnDidDamage: Called when the player pawn in the volume does damage.
  (Utilities|SwitchNetMode
    (:Server
      (bind _asfort_player_pawn_athena (Utilities|Casting|CastToFortPlayerPawnAthena DamagedActor)) ; Actor damaged is a player pawn?
      (LocalVariables|SetDamagedPawn _asfort_player_pawn_athena)
      (if (or _bsuccess _bsuccess_2)
        (Utilities|IsValid _playerstate
          (:Is%20Valid
            ;;[0.04521 0.1314 0.151] Is the instigator pawn in the volume?
            (bind _asfort_player_pawn_athena_2 (Utilities|Casting|CastToFortPlayerPawnAthena (PlayerState|GetPawn _playerstate)))
            (if (Utilities|Array|ContainsItem (Variables|Transient|GetOverlappingPlayerPawns) _asfort_player_pawn_athena_2)
              ;;[0.05723 0.1979 0.1162] Is the damaged pawn in the volume?
              (if (Utilities|Array|ContainsItem (Variables|Transient|GetOverlappingPlayerPawns) (LocalVariables|GetDamagedPawn))
                (Variables|Transient|SetTotalPawnDamageDoneInsideVolume (+ (Variables|Transient|GetTotalPawnDamageDoneInsideVolume) _damage))
                (else
                  (Variables|Transient|SetTotalPawnDamageDoneOutsideVolume (+ _damage (Variables|Transient|GetTotalPawnDamageDoneOutsideVolume))))))))))))

(fn HandlePawnDied
  (params (FortPlayerPawnAthena :type Fort%20Player%20Pawn%20Athena%20Object%20Reference))
  ;[0.1354 0.03243 0.125] HandlePawnDied: Called whenever a player pawn on the Poi dies.
  (Utilities|SwitchNetMode
    (:Server
      ;;[0.1202 0.04357 0.1562] Is dying pawn in the volume?
      (Utilities|IsValid FortPlayerPawnAthena
        (:Is%20Valid
          (if (Utilities|Array|ContainsItem (Variables|Transient|GetOverlappingPlayerPawns) FortPlayerPawnAthena)
            ;;[0.01425 0.04515 0.05729] Is the damage instigating player in the volume?
            (bind _finisherordownerpawn (Default|GetFinisherorDownerPawn FortPlayerPawnAthena))
            (Utilities|IsValid _finisherordownerpawn
              (:Is%20Valid
                (if (Utilities|Array|ContainsItem (Variables|Transient|GetOverlappingPlayerPawns) _finisherordownerpawn)
                  (Math|Integer|IncrementInt (Variables|Transient|GetTotalPawnDeaths)))))))))))

(fn NotifyActorBunkerEntered
  (params (Actor :type Actor%20Object%20Reference))
  ;[0.009605 0.05729 0.05043] NotifyActorBunkerEntered: Track the unique players that are entering the bunker.
  ; This is called from the BGA_WeaponBunkerDoor.
  (Utilities|SwitchNetMode
    (:Server
      (bind _asfort_player_pawn_athena (Utilities|Casting|CastToFortPlayerPawnAthena Actor)) ; Overlapped a player pawn.
      (bind _asfort_player_state_athena (Utilities|Casting|CastToFortPlayerStateAthena (Variables|/Script/Engine.Pawn|GetPlayerState _asfort_player_pawn_athena)))
      (if (Variables|/Script/Engine.PlayerState|GetbIsABot _asfort_player_state_athena)
        (Utilities|Array|AddUnique (Variables|Transient|GetUniquePlayerIdsEnteringBunker_Bot) (PlayerState|GetPlayerUniqueNetID _asfort_player_state_athena))
        (else
          (Utilities|Array|AddUnique (Variables|Transient|GetUniquePlayerIdsEnteringBunker_Human) (PlayerState|GetPlayerUniqueNetID _asfort_player_state_athena)))))))

(fn SetEnableAnalyticsVolume
  (params (bEnable :type Boolean))
  ;[0.01555 0.1213 0.125] SetEnableAnalyticsVolume: Update the collision settings on the box component accordingly.
  (Utilities|SwitchNetMode
    (:Server
      (Collision|SetCollisionEnabled (Variables|Default|GetBox) (select bEnable "QueryOnly" "NoCollision")))))

(fn SubmitAnalytics
  (bind _uniqueplayeridsenteringbunker_bot (Variables|Transient|GetUniquePlayerIdsEnteringBunker_Bot))
  (bind _returnvalue (Utilities|Array|Length _uniqueplayeridsenteringbunker_bot))
  (bind _teams (Variables|Transient|GetTeams))
  (bind _returnvalue_2 (Utilities|Array|Length _teams))
  (bind _uniqueplayerids_human (Variables|Transient|GetUniquePlayerIds_Human))
  (bind _returnvalue_3 (Utilities|Array|Length _uniqueplayerids_human))
  (bind _uniqueplayeridsenteringbunker_human (Variables|Transient|GetUniquePlayerIdsEnteringBunker_Human))
  (bind _returnvalue_4 (Utilities|Array|Length _uniqueplayeridsenteringbunker_human))
  (bind _uniqueplayerids_bot (Variables|Transient|GetUniquePlayerIds_Bot))
  (bind _returnvalue_5 (Utilities|Array|Length _uniqueplayerids_bot))
  (bind _totalpawndeaths (Variables|Transient|GetTotalPawnDeaths))
  (bind _totalpawndamagedoneinsidevolume (Variables|Transient|GetTotalPawnDamageDoneInsideVolume))
  (bind _totalpawndamagedoneoutsidevolume (Variables|Transient|GetTotalPawnDamageDoneOutsideVolume))
  ;[0.04251 0.1354 0.04262] SubmitAnalytics: Fire the analytics!
  (Utilities|SwitchNetMode
    (:Server
      (if (and (not (Replay|GetIsPlayingReplay)) (not (Variables|Transient|GetbAnalyticsSubmitted)))
        (Variables|Transient|SetbAnalyticsSubmitted true)
        (Utilities|FlowControl|Sequence
          (:then_0
            (Utilities|FlowControl|Sequence
              (:then_0
                ;;[0.02121 0.07343 0.08854] At this point, all overlapping players can be considered as players that have left the volume.
                (for _array_element (Variables|Transient|GetOverlappingPlayerPawns)
                  (CallFunction|ComputeTotalDurationforLeavingPlayer _array_element)))
              (:then_1
                (bind _returnvalue_6 (Analytics|ConstructAnalyticsAttribute :AttributeName (CallFunction|CreateAttributeName "UniqueHumanPlayersInVolume") :IntValue _returnvalue_3 :floatValue -1.0 :BoolValue false))
                (Utilities|Array|Add (LocalVariables|GetAnalyticsEventAttributes) _returnvalue_6)
                (Fortnite|FortShippingLog self (Utilities|String|Append (CallFunction|CreateAttributeName "UniqueHumanPlayersInVolume") (Utilities|String|ToString(Integer) _returnvalue_3)))
                (bind _returnvalue_7 (Analytics|ConstructAnalyticsAttribute :AttributeName (CallFunction|CreateAttributeName "UniqueTeamsInVolume") :IntValue _returnvalue_2 :floatValue -1.0 :BoolValue false))
                (Utilities|Array|Add (LocalVariables|GetAnalyticsEventAttributes) _returnvalue_7)
                (Fortnite|FortShippingLog self (Utilities|String|Append (CallFunction|CreateAttributeName "UniqueTeamsInVolume") (Utilities|String|ToString(Integer) _returnvalue_2)))
                (bind _returnvalue_8 (Analytics|ConstructAnalyticsAttribute :AttributeName (CallFunction|CreateAttributeName "TotalDamageDoneInsideVolume") :IntValue -1 :floatValue _totalpawndamagedoneinsidevolume :BoolValue false))
                (Utilities|Array|Add (LocalVariables|GetAnalyticsEventAttributes) _returnvalue_8)
                (Fortnite|FortShippingLog self (Utilities|String|Append (CallFunction|CreateAttributeName "TotalDamageDoneInsideVolume") (Utilities|String|ToString(Float) _totalpawndamagedoneinsidevolume)))
                (bind _time (CallFunction|GetTotalPlayerGameplayTimeinVolume))
                (bind _returnvalue_9 (Analytics|ConstructAnalyticsAttribute :AttributeName (CallFunction|CreateAttributeName "TotalPlayerGametimeTimeInVolume") :IntValue -1 :floatValue _time :BoolValue false))
                (Utilities|Array|Add (LocalVariables|GetAnalyticsEventAttributes) _returnvalue_9)
                (Fortnite|FortShippingLog self (Utilities|String|Append (CallFunction|CreateAttributeName "TotalPlayerGametimeTimeInVolume") (Utilities|String|ToString(Float) _time)))
                (bind _time_2 (CallFunction|GetAveragePlayerGameplayTimeinVolume))
                (bind _returnvalue_10 (Analytics|ConstructAnalyticsAttribute :AttributeName (CallFunction|CreateAttributeName "AveragePlayerGameplayTimeInVolume") :IntValue -1 :floatValue _time_2 :BoolValue false))
                (Utilities|Array|Add (LocalVariables|GetAnalyticsEventAttributes) _returnvalue_10)
                (Fortnite|FortShippingLog self (Utilities|String|Append (CallFunction|CreateAttributeName "AveragePlayerGameplayTimeInVolume") (Utilities|String|ToString(Float) _time_2))))
              (:then_2
                (bind _returnvalue_11 (Analytics|ConstructAnalyticsAttribute :AttributeName (CallFunction|CreateAttributeName "TotalDamageDoneOutsideVolume") :IntValue -1 :floatValue _totalpawndamagedoneoutsidevolume :BoolValue false))
                (Utilities|Array|Add (LocalVariables|GetAnalyticsEventAttributes) _returnvalue_11)
                (Fortnite|FortShippingLog self (Utilities|String|Append (CallFunction|CreateAttributeName "TotalDamageDoneOutsideVolume") (Utilities|String|ToString(Float) _totalpawndamagedoneoutsidevolume)))
                (bind _returnvalue_12 (Analytics|ConstructAnalyticsAttribute :AttributeName (CallFunction|CreateAttributeName "TotalPlayerPawnDeathsInVolume") :IntValue _totalpawndeaths :floatValue -1.0 :BoolValue false))
                (Utilities|Array|Add (LocalVariables|GetAnalyticsEventAttributes) _returnvalue_12)
                (Fortnite|FortShippingLog self (Utilities|String|Append (CallFunction|CreateAttributeName "TotalPlayerPawnDeathsInVolume") (Utilities|String|ToString(Integer) _totalpawndeaths)))
                (bind _returnvalue_13 (Analytics|ConstructAnalyticsAttribute :AttributeName (CallFunction|CreateAttributeName "UniquePlayerBotsInVolume") :IntValue _returnvalue_5 :floatValue -1.0 :BoolValue false))
                (Utilities|Array|Add (LocalVariables|GetAnalyticsEventAttributes) _returnvalue_13)
                (Fortnite|FortShippingLog self (Utilities|String|Append (CallFunction|CreateAttributeName "UniquePlayerBotsInVolume") (Utilities|String|ToString(Integer) _returnvalue_5)))
                (bind _returnvalue_14 (Analytics|ConstructAnalyticsAttribute :AttributeName (CallFunction|CreateAttributeName "UniqueHumanPlayersEnteringBunker") :IntValue _returnvalue_4 :floatValue -1.0 :BoolValue false))
                (Utilities|Array|Add (LocalVariables|GetAnalyticsEventAttributes) _returnvalue_14)
                (Fortnite|FortShippingLog self (Utilities|String|Append (CallFunction|CreateAttributeName "UniqueHumanPlayersEnteringBunker") (Utilities|String|ToString(Integer) _returnvalue_4)))
                (bind _returnvalue_15 (Analytics|ConstructAnalyticsAttribute :AttributeName (CallFunction|CreateAttributeName "UniqueBotPlayersEnteringBunker") :IntValue _returnvalue :floatValue -1.0 :BoolValue false))
                (Utilities|Array|Add (LocalVariables|GetAnalyticsEventAttributes) _returnvalue_15)
                (Fortnite|FortShippingLog self (Utilities|String|Append (CallFunction|CreateAttributeName "UniqueBotPlayersEnteringBunker") (Utilities|String|ToString(Integer) _returnvalue))))))
          (:then_1
            (Analytics|FireAnalyticsEventBlueprintArray "Core.MMObj.WeaponBunker" (LocalVariables|GetAnalyticsEventAttributes))))))))

(fn ConstructionScript)
